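# Niagara anim trail effect for the Street Fashion Summer Female pickaxe, drawing a short-lived ribbon along the weapon as it animates.
# A single ribbon emitter spawns and updates trail particles through shared AnimTrail modules, with bone or socket indices set at spawn and a skeletal mesh exposed as a user parameter.
# Trail alpha fades over life by a cubic curve, scaled by a time-of-day scalar that boosts it slightly during the day.
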NiagaraSystem /BRCosmetics/Weapons/FORT_Melee/Pickaxe_Street_Fashion_Summer_Female/FX/NS_PIckaxe_Street_Fashion_Summer_Female_AnimTrail
dependencies {usedDataInterfaces [&NiagaraDataInterfaceCurve], usedDynamicInputs [], usedModules [&SystemState]}
emitters {Ribbon_AnimTrail {
  emitterUpdateScript [
   &EmitterState()
   &FN_SpawnAnimTrail_v2()
  ]
  particleSpawnScript [
   /Game/Effects/Niagara/Modules/AnimTrail/FN_InitializeAnimTrail/FN_InitializeAnimTrailBoneIndecies(First Bone or Socket Index=1)
   /Game/Effects/Niagara/Modules/AnimTrail/FN_InitializeAnimTrail/FN_InitializeAnimTrail_v3/FN_InitializeAnimTrail(Color:LinearColor=<rgba 1 1 1 1.15>, Lifetime=0.55)
  ]
  particleUpdateScript [
   &ParticleState()
   /Game/Effects/Niagara/Modules/AnimTrail/FN_UpdateAnimTrail/FN_UpdateAnimTrail_v3/FN_UpdateAnimTrail()
   &ScaleColor(
    Scale Alpha=&FloatFromCurve(
     FloatCurve=DataInterface() {propertyValues {Curve {keys [{arriveTangent=-2.1531, interpMode=RCIM_Cubic, leaveTangent=-2.1531, tangentMode=RCTM_User, time=0, value=1}, <at 1 0 RCIM_Cubic>]}}}
     Scale Curve=&FNTimeOfDayScalar(Night Scale=1, Morning Scale=1, Day Scale=1.15, Evening Scale=1)
    )
    ScaleA=true
    ScaleRGB=true
   )
  ]
  rendererClasses [&NiagaraRibbonRendererProperties]
  renderers [{rendererClass=&NiagaraRibbonRendererProperties}]
 }}
summary {
 systemName=NS_PIckaxe_Street_Fashion_Summer_Female_AnimTrail
 emitters [{emitterName=Ribbon_AnimTrail, rendererClasses [&NiagaraRibbonRendererProperties]}]
 userVariables [
  User.RibbonIDOffset {type=NiagaraInt32}
  User.SkeletalMesh {
   type=NiagaraDataInterfaceSkeletalMesh
   defaultValue {
    struct=NiagaraExt_VariableValue_DataInterface
    value {dataInterface=&self:NiagaraDataInterfaceSkeletalMesh_2}
   }
  }
 ]
}
